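; Parent Blueprint for the ornate fine art frame props, exposing canvas scale and position adjustments.
; The construction script writes global, horizontal and vertical scale, plus horizontal and vertical position, into default Custom Primitive Data floats on the static mesh component.
; This supplies per-instance values to the material without dynamic material instances.
; BeginPlay, overlap and Tick events hold no active logic.
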(event EventBeginPlay
  (|CallParentFunction) ; This node is disabled and will not be called.\nDrag off pins to build functionality.
)

(event Collision|EventActorBeginOverlap
  (params (OtherActor :type Actor%20Object%20Reference)))

(event EventTick
  (params (DeltaSeconds :type Float%20%28single-precision%29)))

(fn ConstructionScript
  (|CallParentFunction)
  (Rendering|Material|SetDefaultCustomPrimitiveDataFloat (Variables|Debug|GetStaticMeshComponent) 0 (Variables|CanvasAdjustments|GetScaleGlobal))
  (Rendering|Material|SetDefaultCustomPrimitiveDataFloat (Variables|Debug|GetStaticMeshComponent) 1 (Variables|CanvasAdjustments|GetScaleHorizontal))
  (Rendering|Material|SetDefaultCustomPrimitiveDataFloat (Variables|Debug|GetStaticMeshComponent) 2 (Variables|CanvasAdjustments|GetScaleVertical))
  (Rendering|Material|SetDefaultCustomPrimitiveDataFloat (Variables|Debug|GetStaticMeshComponent) 3 (Variables|CanvasAdjustments|GetPositionHorizontal))
  (Rendering|Material|SetDefaultCustomPrimitiveDataFloat (Variables|Debug|GetStaticMeshComponent) 4 (Variables|CanvasAdjustments|GetPositionVertical)))
